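# Character material for the Pliant backpack cosmetic that recolors the emissive glow with a primary color parameter.
# It chains the shared PreFX and CharacterSkinFX functions, then desaturates the resulting emissive, raises it to a power and multiplies by the tint.
# The saturated result is scaled by a multiplier with a separate mobile value, written back into the material attributes, then passed through the character PostFX and cinematic post functions.
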Material /BRCosmetics/Accessories/FORT_Backpacks/F_MED_Pliant/Material/M_Pliant_Backpack
graph {
 properties {
  bAutomaticallySetUsageInEditor=false
  bUseMaterialAttributes=true
  bUsedWithMeshDeformer=true
  bUsedWithMorphTargets=true
  bUsedWithSkeletalMesh=true
  pixelDepthOffsetMode=PDOM_Legacy
  refractionMethod=RM_IndexOfRefraction
  shadingModel=MSM_FromMaterialExpression
  subsurfaceProfile=&SS_HeroSkin_02
  usedShadingModels=Default Lit | Subsurface | Clear Coat | Subsurface Profile | Cloth | Eye
 }
 nodes [
  n1 = MaterialFunctionCall() {materialFunction=&BaseFN_CharacterFunction_PreFX}
  n2 = GetMaterialAttributes(None=MaterialFunctionCall(
    PreskinnedPosition=@n1.Pre skinned Local Position
    Gmap=@n1.Pre skinned Local Position
    MaterialAttributes=@n1.Pre skinned Local Position
    AO=@n1.Pre skinned Local Position
   ) {materialFunction=&CharacterSkinFX}.Result) {attributeGetTypes [#1, #2]}
  n3 = Desaturation(None=@n2.Emissive Color)
  n4 = ScalarParameter() {defaultValue=1, group=Pliant, parameterName=Pliant_BackpackEmsTintPow, sliderMax=2}
  n5 = Saturate(None=Multiply(
    A=Power(Base=@n3, Exp=@n4)
    B=VectorParameter() {group=Pliant, parameterName=Pliant_PrimaryColor, defaultValue=<rgba 0.7552 0.1731 0.5466 1>}.RGB
   ))
  n6 = ScalarParameter() {defaultValue=1, group=Pliant, parameterName=Pliant_BackpackTintEmsMult}
  n7 = ShadingPathSwitch(Default=@n6, Deferred=@n6, Mobile=ScalarParameter() {defaultValue=1, group=Pliant, parameterName=Pliant_BackpackTintEmsMult_Mobile}) {desc=Migrated from FeatureLevelSwitch}
  n8 = MaterialFunctionCall(
   AO=@n1.AO
   Distance Blend=@n1.AO
   Fresnel=@n1.AO
   Gmap=@n1.AO
   Pre Skinned Local Position=@n1.AO
   Material=SetMaterialAttributes!(MaterialAttributes=@n2.MaterialAttributes, Emissive Color=Multiply(A=@n5, B=@n7), Base Color) {attributeSetTypes [#1, #2]}
  ) {materialFunction=&BaseFN_CharacterFunction_PostFX}
  MaterialFunctionCall(In=@n8.Result) {materialFunction=&CINE_MF_CharPostFX}
 ]
}
