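# Physics asset for the F_MED_Pliant cosmetic backpack mesh, defining collision and joint limits for the pack and its three dynamic cables.
# A large rotated capsule covers the pack body, and a thin capsule sits on each cable bone (A, B, C).
# Each cable is constrained directly to the pack rather than chained, sharing moderate asymmetric swing limits and very tight twist.
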PhysicsAsset /BRCosmetics/Accessories/FORT_Backpacks/F_MED_Pliant/Meshes/F_MED_Pliant_Pack_Physics
bodies {
 dyn_cable_A [{shapeType=Capsule, length=5.9587, radius=0.505, center=<6.6 0 0>}]
 dyn_cable_B [{shapeType=Capsule, length=6.9936, radius=0.3, center=<2.9 0 0>}]
 dyn_cable_C [{shapeType=Capsule, length=8.8277, radius=0.3, center=<3.3 0 0>}]
 pack [{shapeType=Capsule, length=17.2557, radius=16.6202, center=<-0.1871 18.4718 -23.9497>, rotation=<rot 90 0 0>}]
}
constraints [
 {bone1Name=dyn_cable_B, bone2Name=pack, swing1LimitDegrees=40, swing2LimitDegrees=20, twistLimitDegrees=2}
 {bone1Name=dyn_cable_A, bone2Name=pack, swing1LimitDegrees=40, swing2LimitDegrees=20, twistLimitDegrees=2}
 {bone1Name=dyn_cable_C, bone2Name=pack, swing1LimitDegrees=40, swing2LimitDegrees=20, twistLimitDegrees=2}
]
